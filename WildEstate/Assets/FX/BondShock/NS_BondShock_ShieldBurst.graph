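# Niagara system for a brief blue BondShock shield burst that tracks points sampled from the User.SkeletalMesh input.
# Each of three emitters spawns in an instantaneous burst: Empty is a fading light, NE_Empty an alpha-flashed sprite plus mesh.
# Flares stacks four sprite renderers, with custom attributes animating anamorphic, corona and distant-glow colors and sizes by curve.
# The main flare swells and peaks late in its lifetime, then fades.
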NiagaraSystem /WildEstate/Assets/FX/BondShock/NS_BondShock_ShieldBurst
dependencies {usedDataInterfaces [&NiagaraDataInterfaceCurve], usedDynamicInputs [], usedModules [&SystemState]}
emitters {
 Empty {
  emitterUpdateScript [
   &EmitterState()
   &SpawnBurst_Instantaneous(Use Spawn Probability=false, UseLoopCountLimit=false)
  ]
  particleSpawnScript [&InitializeParticle(Color=<rgba 0.026 0.377 1 1>, Lifetime=0.2, Position Offset=<0 0 -50>, UsePositionOffset=true, Write Material Random=true)]
  particleUpdateScript [
   &ParticleState()
   &SkeletalMeshLocation(Bone Sampling Mode=::NewEnumerator4, MaskSpawnGroup=false, Sampled Position Offset=<20 0 0>, Skeletal Mesh=$User.SkeletalMesh)
   /SetVariables(Particles.LightExponent=2, Particles.LightRadius=150) {isSetParameters=true}
   &ScaleColor(
    Scale RGB=&VectorFromFloat(Value=&FloatFromCurve(
      FloatCurve=DataInterface() {propertyValues {Curve {keys [
          {interpMode=RCIM_Cubic, leaveTangentWeight=0.4714, tangentMode=RCTM_User, time=0, value=1}
          {arriveTangentWeight=0.4714, interpMode=RCIM_Cubic, tangentMode=RCTM_User, time=1, value=0}
         ]}}}
      Scale Curve=5
     ))
    ScaleA=true
    ScaleRGB=true
   )
  ]
  rendererClasses [&NiagaraLightRendererProperties]
  renderers [{rendererClass=&NiagaraLightRendererProperties, rendererIndex=0}]
 }
 Flares {
  emitterUpdateScript [
   &EmitterState()
   &SpawnBurst_Instantaneous(Use Spawn Probability=false, UseLoopCountLimit=false)
  ]
  particleSpawnScript [
   &InitializeParticle(
    Color=<rgba 0.0156 1.0104 3 1>
    Lifetime=0.2
    Sprite Rotation Angle=0.2
    Sprite Rotation Mode=::NewEnumerator1
    Sprite Size Mode=::NewEnumerator3
    Uniform Sprite Size=150
    UsePositionOffset=false
    Write Material Random=true
   )
   &CameraOffset(Camera Offset Amount=75)
   /SetVariables(Particles.AnamorphicSize:Vector2f=<240 7>) {isSetParameters=true}
  ]
  particleUpdateScript [
   &ParticleState()
   &SkeletalMeshLocation(Bone Sampling Mode=::NewEnumerator4, MaskSpawnGroup=false, Sampled Position Offset=<20 0 0>, Skeletal Mesh=$User.SkeletalMesh)
   &ScaleColor(
    Scale Alpha=&FloatFromCurve(FloatCurve=DataInterface() {propertyValues {Curve {keys [
         {arriveTangent=0.6364, interpMode=RCIM_Cubic, leaveTangent=0.6364, time=0, value=0}
         {arriveTangent=1.4286, interpMode=RCIM_Cubic, leaveTangent=1.4286, time=0.55, value=0.35}
         {arriveTangent=-8.2069, interpMode=RCIM_Cubic, leaveTangent=-8.2069, tangentMode=RCTM_User, time=0.7, value=1}
         {arriveTangent=-0.0405, interpMode=RCIM_Cubic, leaveTangent=-0.0405, tangentMode=RCTM_User, time=1, value=0}
        ]}}})
    ScaleA=true
    ScaleRGB=true
   )
   &ScaleSpriteSize(Uniform Curve Sprite Scale=DataInterface() {propertyValues {Curve {keys [
        {interpMode=RCIM_Cubic, leaveTangentWeight=0.1863, tangentMode=RCTM_User, time=0, value=0.4}
        {
         arriveTangent=0.3825
         arriveTangentWeight=0.2415
         interpMode=RCIM_Cubic
         leaveTangent=0.3825
         leaveTangentWeight=0.0659
         tangentMode=RCTM_User
         time=0.55
         value=0.5
        }
        {
         arriveTangent=-1.9384
         arriveTangentWeight=0.0547
         interpMode=RCIM_Cubic
         leaveTangent=-1.9384
         leaveTangentWeight=0.1094
         tangentMode=RCTM_User
         time=0.7
         value=1
        }
        {arriveTangentWeight=0.1414, interpMode=RCIM_Cubic, tangentMode=RCTM_User, time=1, value=0.7}
       ]}}})
   /SetVariables(
    Particles.Anamorphic Color:LinearColor=&MakeLinearColorFromVectorAndFloat(
     "Vector (RGB)"=&MakeVectorFromLinearColorRGB(Color=<rgba 0.0156 1.0104 3 1>)
     "Float (Alpha)"=&Multiply_Float(
      A=0.2
      B=&FloatFromCurve(FloatCurve=DataInterface() {propertyValues {Curve {keys [
           {arriveTangent=0.6364, interpMode=RCIM_Cubic, leaveTangent=0.6364, time=0, value=0}
           {arriveTangent=1.4286, interpMode=RCIM_Cubic, leaveTangent=1.4286, time=0.55, value=0.35}
           {arriveTangent=-8.2069, interpMode=RCIM_Cubic, leaveTangent=-8.2069, tangentMode=RCTM_User, time=0.7, value=1}
           {arriveTangent=-0.0405, interpMode=RCIM_Cubic, leaveTangent=-0.0405, tangentMode=RCTM_User, time=1, value=0}
          ]}}})
     )
    )
    Particles.CoronaColor:LinearColor=&MakeLinearColorFromVectorAndFloat(
     "Vector (RGB)"=&MakeVectorFromLinearColorRGB(Color=<rgba 0.0156 1.0104 3 1>)
     "Float (Alpha)"=&FloatFromCurve(FloatCurve=DataInterface() {propertyValues {Curve {keys [
          {interpMode=RCIM_Cubic, leaveTangentWeight=0.3578, tangentMode=RCTM_User, time=0, value=0}
          {arriveTangent=0.6909, interpMode=RCIM_Cubic, leaveTangent=0.6909, tangentMode=RCTM_User, time=0.5, value=0.95}
          {arriveTangentWeight=0.1431, interpMode=RCIM_Cubic, leaveTangentWeight=0.2147, tangentMode=RCTM_User, time=0.7, value=1}
          {arriveTangentWeight=0.348, interpMode=RCIM_Cubic, tangentMode=RCTM_User, time=1, value=0}
         ]}}})
    )
    Particles.CoronaDynParameters:Vector4f=&MakeVector4(
     X=&FloatFromCurve(FloatCurve=DataInterface() {propertyValues {Curve {keys [
          {interpMode=RCIM_Cubic, leaveTangentWeight=0.1904, tangentMode=RCTM_User, time=0, value=1.4}
          {
           arriveTangent=-1.2888
           arriveTangentWeight=0.2753
           interpMode=RCIM_Cubic
           leaveTangent=-1.2888
           leaveTangentWeight=0.0918
           tangentMode=RCTM_User
           time=0.525
           value=1.175
          }
          {arriveTangent=-1.9474, interpMode=RCIM_Cubic, leaveTangent=-1.9474, time=0.7, value=0.55}
          {arriveTangentWeight=0.1414, interpMode=RCIM_Cubic, tangentMode=RCTM_User, time=1, value=0.25}
         ]}}})
     Y=8
     Z=0.6
     W=1
    )
    Particles.CoronaSize:Vector2f=&Vector2DFromFloat(Value=&Multiply_Float(
      A=75
      B=&FloatFromCurve(FloatCurve=DataInterface() {propertyValues {Curve {keys [
           {interpMode=RCIM_Cubic, leaveTangentWeight=0.3578, tangentMode=RCTM_User, time=0, value=0}
           {
            arriveTangent=1.1981
            arriveTangentWeight=0.2906
            interpMode=RCIM_Cubic
            leaveTangent=1.1981
            leaveTangentWeight=0.1163
            tangentMode=RCTM_User
            time=0.5
            value=0.95
           }
           {arriveTangentWeight=0.1431, interpMode=RCIM_Cubic, leaveTangentWeight=0.2147, tangentMode=RCTM_User, time=0.7, value=1}
          ]}}})
     ))
    Particles.DistanceGlowSize:Vector2f=&Vector2DFromFloat(Value=&Multiply_Float(
      A=400
      B=&FloatFromCurve(FloatCurve=DataInterface() {propertyValues {Curve {keys [
           {interpMode=RCIM_Cubic, leaveTangentWeight=0.348, tangentMode=RCTM_User, time=0, value=0.7}
           {arriveTangent=0.7107, arriveTangentWeight=0.348, interpMode=RCIM_Cubic, leaveTangent=0.7107, tangentMode=RCTM_User, time=1, value=1}
          ]}}})
     ))
    Particles.DistantGlowColor:LinearColor=&MakeLinearColorFromVectorAndFloat(
     "Vector (RGB)"=&MakeVectorFromLinearColorRGB(Color=<rgba 0.026 0.377 1 1>)
     "Float (Alpha)"=&FloatFromCurve(FloatCurve=DataInterface() {propertyValues {Curve {keys [
          {arriveTangent=9.0725, interpMode=RCIM_Cubic, leaveTangent=9.0725, tangentMode=RCTM_User, time=0, value=0}
          {arriveTangentWeight=0.0905, interpMode=RCIM_Cubic, leaveTangentWeight=0.8147, tangentMode=RCTM_User, time=0.1, value=0.3}
          {arriveTangent=-0.0405, interpMode=RCIM_Cubic, leaveTangent=-0.0405, tangentMode=RCTM_User, time=1, value=0}
         ]}}})
    )
   ) {isSetParameters=true}
  ]
  rendererClasses [&NiagaraSpriteRendererProperties]
  renderers [
   {rendererClass=&NiagaraSpriteRendererProperties}
   {rendererClass=&NiagaraSpriteRendererProperties, rendererIndex=1}
   {rendererClass=&NiagaraSpriteRendererProperties, rendererIndex=2}
   {rendererClass=&NiagaraSpriteRendererProperties, rendererIndex=3}
  ]
 }
 NE_Empty {
  emitterUpdateScript [
   &EmitterState()
   &SpawnBurst_Instantaneous(Use Spawn Probability=false, UseLoopCountLimit=false)
   &SpawnRate(SpawnRate=20, Use Spawn Probability=false) {enabled=false}
  ]
  particleSpawnScript [
   /Niagara/Modules/Spawn/Initialization/InitializeParticle(
    Color=<rgba 0.1562 10.1041 30 1>
    Module.Write Color=true
    Module.Write Lifetime=true
    Module.Write Mass=true
    Module.Write Material Random=true
    Module.Write Position=false
    Module.Write Scale=true
    Module.Write SpriteRotation=false
    Module.Write SpriteSize=true
    Position=&SimulationPosition()
    Sprite Size=<5 5>
   )
   &InitialMeshOrientation(Mesh Axis To Orient=<0 0 1>, Mesh Orientation Mode=::NewEnumerator2, Use Rotation Vector=false)
  ]
  particleUpdateScript [
   &ParticleState(Kill Particles When Lifetime Has Elapsed=false, Let Infinitely Lived Particles Die When Emitter Deactivates=true)
   &SkeletalMeshLocation(Bone Sampling Mode=::NewEnumerator4, MaskSpawnGroup=false, Skeletal Mesh=$User.SkeletalMesh)
   &SolveForcesAndVelocity(Clamp Velocity=false, Limit Acceleration=false)
   &DynamicMaterialParameters(
    Index 0 Param 1=1
    Index 0 Param 3=1
    Module.Index0Float0Write=true
    Module.Index0Float1Write=false
    Module.Index0Float2Write=false
    Module.Index0Float3Write=false
   )
   &ScaleColor(
    Scale Alpha=&FloatFromCurve(FloatCurve=DataInterface() {propertyValues {Curve {
        defaultValue=1
        keys [
         {arriveTangent=9.0725, interpMode=RCIM_Cubic, leaveTangent=9.0725, tangentMode=RCTM_User, time=0, value=0}
         {arriveTangent=-2.5249, interpMode=RCIM_Cubic, leaveTangent=-2.5249, tangentMode=RCTM_User, time=0.1408, value=0.9847}
         {arriveTangent=-0.0405, interpMode=RCIM_Cubic, leaveTangent=-0.0405, tangentMode=RCTM_User, time=1, value=0}
        ]
       }}})
    ScaleA=true
    ScaleRGB=true
   )
  ]
  rendererClasses [&NiagaraSpriteRendererProperties, &NiagaraMeshRendererProperties]
  renderers [
   {rendererClass=&NiagaraSpriteRendererProperties}
   {rendererClass=&NiagaraMeshRendererProperties, rendererIndex=1}
  ]
 }
}
summary {
 systemName=NS_BondShock_ShieldBurst
 emitters [
  {emitterName=Flares, rendererClasses [&NiagaraSpriteRendererProperties]}
  {emitterName=Empty, rendererClasses [&NiagaraLightRendererProperties]}
  {emitterName=NE_Empty, rendererClasses [&NiagaraSpriteRendererProperties, &NiagaraMeshRendererProperties]}
 ]
 userVariables [User.SkeletalMesh {
   type=NiagaraDataInterfaceSkeletalMesh
   defaultValue {
    struct=NiagaraExt_VariableValue_DataInterface
    value {dataInterface=&self:NiagaraDataInterfaceSkeletalMesh_0}
   }
  }]
}
